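# Physics asset for the Assemble_L pickaxe axe mesh, giving physics bodies to its head and its dangling tube and ribbon bones.
# A head capsule acts as the hub, with four identically sized smaller capsules for the high and low tubes and ribbons.
# Each small body is constrained directly to the head with modest swing limits and a near-zero twist limit.
# The high tube bone is spelled "dyb_tube_high", unlike the "dyn_" prefix on the others.
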PhysicsAsset /BRCosmetics/Weapons/FORT_Melee/Pickaxe_Assemble_L/Meshes/Assemble_L_Axe_Physics
bodies {
 dyb_tube_high [{shapeType=Capsule, length=5.0048, radius=2.0694, center=<4.4129 0 0>, rotation=<rot 90 0 0>}]
 dyn_head [{shapeType=Capsule, length=6.8574, radius=4.8334}]
 dyn_ribbon_high [{shapeType=Capsule, length=5.0048, radius=2.0694, center=<4.4129 0 0>, rotation=<rot 90 0 0>}]
 dyn_ribbon_low [{shapeType=Capsule, length=5.0048, radius=2.0694, center=<4.4129 0 0>, rotation=<rot 90 0 0>}]
 dyn_tube_low [{shapeType=Capsule, length=5.0048, radius=2.0694, center=<4.4129 0 0>, rotation=<rot 90 0 0>}]
}
constraints [
 {bone1Name=dyn_tube_low, bone2Name=dyn_head, swing1LimitDegrees=25, swing2LimitDegrees=25, twistLimitDegrees=1}
 {bone1Name=dyn_ribbon_low, bone2Name=dyn_head, swing1LimitDegrees=25, swing2LimitDegrees=25, twistLimitDegrees=1}
 {bone1Name=dyn_ribbon_high, bone2Name=dyn_head, swing1LimitDegrees=20, swing2LimitDegrees=20, twistLimitDegrees=1}
 {bone1Name=dyb_tube_high, bone2Name=dyn_head, swing1LimitDegrees=25, swing2LimitDegrees=25, twistLimitDegrees=1}
]
